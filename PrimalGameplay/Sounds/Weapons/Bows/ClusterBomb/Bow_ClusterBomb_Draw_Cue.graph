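# Sound cue for drawing the Cluster Bomb bow, reusing the metal bow's default draw samples.
# A mixer layers a random pick of three draw one-shots with a looping tension sound.
# An Enveloper raises the loop's volume and pitch over several seconds, and a Modulator scales it by fixed volume.
# It uses reload attenuation, the weapon foley class, and global and local bow concurrency, with debug enabled.
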SoundCue /PrimalGameplay/Sounds/Weapons/Bows/ClusterBomb/Bow_ClusterBomb_Draw_Cue
root=Mixer_1
nodes [
 {
  id=Enveloper_1
  children [WavePlayer_0]
  properties {
   pitchCurve {editorCurveData {keys [<at 0 0.9>,<at 10 1.2>]}}
   volumeCurve {editorCurveData {keys [
      {arriveTangent=0.3293, interpMode=RCIM_Cubic, leaveTangent=0.3293, tangentMode=RCTM_User, time=0.0031, value=0.0793}
      {arriveTangent=-0.0494, interpMode=RCIM_Cubic, leaveTangent=-0.0494, tangentMode=RCTM_User, time=5, value=1}
     ]}}
  }
 }
 {id=Mixer_1, children [Random_1,Modulator_0], properties {inputVolume [0.7,1]}}
 {id=Modulator_0, children [Enveloper_1], properties {volumeMax=0.7, volumeMin=0.7}}
 {id=Random_1, children [WavePlayer_4,WavePlayer_5,WavePlayer_6], properties {weights [1,1,1]}}
 {id=WavePlayer_0, properties {bLooping=true, soundWaveAssetPtr=/PrimalGameplay/Sounds/Weapons/Bows/Metal/Bow_Default_Draw_Tension_Loop}}
 {id=WavePlayer_4, properties {soundWaveAssetPtr=/PrimalGameplay/Sounds/Weapons/Bows/Metal/Bow_Default_Draw_01}}
 {id=WavePlayer_5, properties {soundWaveAssetPtr=/PrimalGameplay/Sounds/Weapons/Bows/Metal/Bow_Default_Draw_02}}
 {id=WavePlayer_6, properties {soundWaveAssetPtr=/PrimalGameplay/Sounds/Weapons/Bows/Metal/Bow_Default_Draw_03}}
]
properties {
 attenuationSettings=&Fort_Weapon_Reload_Att
 bDebug=true
 maxDistance=2250
 soundClassObject=&Fort_WeaponFoley
 volumeMultiplier=0.501
 concurrencySet [/Game/Sounds/Concurrency_Rules/Bow_Global_Conc, /Game/Sounds/Concurrency_Rules/Bow_Local_Conc]
}
