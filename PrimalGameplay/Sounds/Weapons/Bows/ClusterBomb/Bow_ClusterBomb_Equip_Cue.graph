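# Sound cue for equipping the Cluster Bomb bow, routed as weapon foley with reload attenuation and global and local bow concurrency.
# A mixer layers a random, modulated Stink bow equip sound, a briefly delayed random, modulated Stink bow reload knock, and a boosted Cluster Bomb-specific layer.
# Only that last layer is unique to this bow, as the other waves are reused from the Stink bow.
# The cue has its debug flag enabled.
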SoundCue /PrimalGameplay/Sounds/Weapons/Bows/ClusterBomb/Bow_ClusterBomb_Equip_Cue
root=Mixer_0
nodes [
 {id=Delay_0, children [Modulator_1], properties {delayMax=0.4, delayMin=0.4}}
 {id=Mixer_0, children [Modulator_0,Delay_0,WavePlayer_9], properties {inputVolume [1,1,1.5]}}
 {id=Modulator_0, children [Random_0]}
 {id=Modulator_1, children [Random_1]}
 {id=Random_0, children [WavePlayer_1,WavePlayer_2,WavePlayer_3,WavePlayer_8], properties {weights [1,1,1,1]}}
 {id=Random_1, children [WavePlayer_4,WavePlayer_5,WavePlayer_6,WavePlayer_7], properties {weights [1,1,1,1]}}
 {id=WavePlayer_1, properties {soundWaveAssetPtr=/PrimalGameplay/Sounds/Weapons/Bows/Stink/Bow_Stink_Equip_Default_01}}
 {id=WavePlayer_2, properties {soundWaveAssetPtr=/PrimalGameplay/Sounds/Weapons/Bows/Stink/Bow_Stink_Equip_Default_02}}
 {id=WavePlayer_3, properties {soundWaveAssetPtr=/PrimalGameplay/Sounds/Weapons/Bows/Stink/Bow_Stink_Equip_Default_03}}
 {id=WavePlayer_4, properties {soundWaveAssetPtr=/PrimalGameplay/Sounds/Weapons/Bows/Stink/Bow_Stink_Reload_Knock_01}}
 {id=WavePlayer_5, properties {soundWaveAssetPtr=/PrimalGameplay/Sounds/Weapons/Bows/Stink/Bow_Stink_Reload_Knock_02}}
 {id=WavePlayer_6, properties {soundWaveAssetPtr=/PrimalGameplay/Sounds/Weapons/Bows/Stink/Bow_Stink_Reload_Knock_03}}
 {id=WavePlayer_7, properties {soundWaveAssetPtr=/PrimalGameplay/Sounds/Weapons/Bows/Stink/Bow_Stink_Reload_Knock_04}}
 {id=WavePlayer_8, properties {soundWaveAssetPtr=/PrimalGameplay/Sounds/Weapons/Bows/Stink/Bow_Stink_Equip_Default_04}}
 {id=WavePlayer_9, properties {soundWaveAssetPtr=&/Bow_ClusterBomb_Equip_Layer_01}}
]
properties {
 attenuationSettings=&Fort_Weapon_Reload_Att
 bDebug=true
 maxDistance=2250
 soundClassObject=&Fort_WeaponFoley
 volumeMultiplier=0.251
 concurrencySet [/Game/Sounds/Concurrency_Rules/Bow_Global_Conc, /Game/Sounds/Concurrency_Rules/Bow_Local_Conc]
}
