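# First-person sound cue for firing the cluster bomb bow at full tension.
# A mixer layers three randomized, modulated stacks: reused metal bow release shots, plus slightly delayed whoosh and beep layers mixed louder than the bow.
# The whoosh layer gets randomized upward pitch variation.
# First-person projectile attenuation, a ranged-fire sound class and first-person gunfire concurrency govern playback.
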SoundCue /PrimalGameplay/Sounds/Weapons/Bows/ClusterBomb/Bow_ClusterBomb_Fire_Full_Tension_1P_Cue
root=Mixer_0
nodes [
 {id=Delay_0, children [Modulator_0], properties {delayMax=0.1, delayMin=0.1}}
 {id=Delay_1, children [Modulator_3], properties {delayMax=0.1, delayMin=0.1}}
 {id=Mixer_0, children [Modulator_2,Delay_1,Delay_0], properties {inputVolume [0.6,1,1.3]}}
 {id=Modulator_0, children [Random_2]}
 {id=Modulator_2, children [Random_0]}
 {id=Modulator_3, children [Random_4], properties {pitchMax=1.25, pitchMin=1}}
 {id=Random_0, children [WavePlayer_0,WavePlayer_1,WavePlayer_2], properties {weights [1,1,1]}}
 {id=Random_2, children [WavePlayer_6, WavePlayer_7, WavePlayer_8, WavePlayer_9, WavePlayer_10], properties {weights [1,1,1,1,1]}}
 {
  id=Random_4
  children [WavePlayer_19, WavePlayer_20, WavePlayer_21, WavePlayer_22, WavePlayer_23, WavePlayer_24]
  properties {weights [1,1,1,1,1,1]}
 }
 {id=WavePlayer_0, properties {soundWaveAssetPtr=/PrimalGameplay/Sounds/Weapons/Bows/Metal/Bow_Metal_Fire_Full_Tension_01}}
 {id=WavePlayer_1, properties {soundWaveAssetPtr=/PrimalGameplay/Sounds/Weapons/Bows/Metal/Bow_Metal_Fire_Full_Tension_02}}
 {
  id=WavePlayer_10
  properties {soundWaveAssetPtr=&/Bow_ClusterBomb_Fire_Full_Tension_Close_Beeps_05}
 }
 {
  id=WavePlayer_19
  properties {soundWaveAssetPtr=&/Bow_ClusterBomb_Fire_Full_Tension_Close_Whosh_01}
 }
 {id=WavePlayer_2, properties {soundWaveAssetPtr=/PrimalGameplay/Sounds/Weapons/Bows/Metal/Bow_Metal_Fire_Full_Tension_03}}
 {
  id=WavePlayer_20
  properties {soundWaveAssetPtr=&/Bow_ClusterBomb_Fire_Full_Tension_Close_Whosh_02}
 }
 {
  id=WavePlayer_21
  properties {soundWaveAssetPtr=&/Bow_ClusterBomb_Fire_Full_Tension_Close_Whosh_03}
 }
 {
  id=WavePlayer_22
  properties {soundWaveAssetPtr=&/Bow_ClusterBomb_Fire_Full_Tension_Close_Whosh_04}
 }
 {
  id=WavePlayer_23
  properties {soundWaveAssetPtr=&/Bow_ClusterBomb_Fire_Full_Tension_Close_Whosh_05}
 }
 {
  id=WavePlayer_24
  properties {soundWaveAssetPtr=&/Bow_ClusterBomb_Fire_Full_Tension_Close_Whosh_06}
 }
 {
  id=WavePlayer_6
  properties {soundWaveAssetPtr=&/Bow_ClusterBomb_Fire_Full_Tension_Close_Beeps_01}
 }
 {
  id=WavePlayer_7
  properties {soundWaveAssetPtr=&/Bow_ClusterBomb_Fire_Full_Tension_Close_Beeps_02}
 }
 {
  id=WavePlayer_8
  properties {soundWaveAssetPtr=&/Bow_ClusterBomb_Fire_Full_Tension_Close_Beeps_03}
 }
 {
  id=WavePlayer_9
  properties {soundWaveAssetPtr=&/Bow_ClusterBomb_Fire_Full_Tension_Close_Beeps_04}
 }
]
properties {
 attenuationSettings=&Fort_Weapon_Projectile_FirstPerson_Athena_Att
 maxDistance=2097152
 soundClassObject=&Fort_WeaponRanged_Fire_1P_PSM
 volumeMultiplier=0.9
 attenuationOverrides {falloffDistance=5000, attenuationShapeExtents {x=1000}}
 concurrencySet [&Weapon_Gunfire_FirstPerson_Conc]
}
