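# Sound cue for the cluster bomb bow's confirmation beep, named for reaching full tension.
# A single wave player feeds the output with the "B" beep wave, with no randomization or mixing.
# It plays in the weapon foley sound class with a boosted volume and borrows a third-person weapon fuse idle attenuation.
# Global and local bow concurrency rules limit overlapping instances, and the debug flag is left on.
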SoundCue /PrimalGameplay/Sounds/Weapons/Bows/ClusterBomb/Bow_ClusterBomb_Full_Tension_Confirm_Beep_Cue
root=WavePlayer_1
nodes [{id=WavePlayer_1, properties {soundWaveAssetPtr=&/Bow_ClusterBomb_Full_Tension_Confirm_Beep_B}}]
properties {
 attenuationSettings=/Game/Sounds/Attenuation/Fort_Weapon_Fuse_Idle_3P_Att
 bDebug=true
 maxDistance=2050
 soundClassObject=&Fort_WeaponFoley
 volumeMultiplier=2.506
 concurrencySet [/Game/Sounds/Concurrency_Rules/Bow_Global_Conc, /Game/Sounds/Concurrency_Rules/Bow_Local_Conc]
}
