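# Sound cue for the Cluster Bomb bow's full-tension confirm loop, judging by its name, classed as weapon foley.
# A single wave player feeds a one-input mixer behind a short fixed delay, with no looping node or looping flag in the cue.
# It plays at reduced volume with weapon-reload attenuation, is limited by global and local bow concurrency rules, and has its debug flag enabled.
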SoundCue /PrimalGameplay/Sounds/Weapons/Bows/ClusterBomb/Bow_ClusterBomb_Full_Tension_Confirm_Loop_Cue
root=Delay_0
nodes [
 {id=Delay_0, children [Mixer_0], properties {delayMax=0.05, delayMin=0.05}}
 {id=Mixer_0, children [WavePlayer_0], properties {inputVolume [1]}}
 {id=WavePlayer_0, properties {soundWaveAssetPtr=&/Bow_ClusterBomb_Full_Tension_Confirm_01}}
]
properties {
 attenuationSettings=&Fort_Weapon_Reload_Att
 bDebug=true
 maxDistance=2250
 soundClassObject=&Fort_WeaponFoley
 volumeMultiplier=0.501
 concurrencySet [/Game/Sounds/Concurrency_Rules/Bow_Global_Conc, /Game/Sounds/Concurrency_Rules/Bow_Local_Conc]
}
