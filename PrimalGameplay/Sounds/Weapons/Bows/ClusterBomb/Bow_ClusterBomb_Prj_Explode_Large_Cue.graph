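# Sound cue for the large explosion of the cluster bomb bow projectile, limited by global and local bow concurrency rules.
# A mixer layers close, mid and distant branches, each picking a random wave, varying its pitch and applying its own explosion attenuation.
# The close layer uses a third-person weapon-fire sound class, while mid and distant share a distant-fire class.
# The distant layer adds a short random delay, and the cue has its debug flag on.
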SoundCue /PrimalGameplay/Sounds/Weapons/Bows/ClusterBomb/Bow_ClusterBomb_Prj_Explode_Large_Cue
root=Mixer_5
nodes [
 {id=Attenuation_4, children [Modulator_5], properties {attenuationSettings=/Game/Sounds/Attenuation/Fort_Explosion_Medium_Athena_Att}}
 {
  id=Attenuation_6
  children [Modulator_3]
  properties {
   attenuationSettings=/Game/Sounds/Attenuation/Fort_Explosion_Close_Athena_Att
   attenuationOverrides {falloffDistance=900, attenuationShapeExtents {x=100}}
  }
 }
 {
  id=Attenuation_8
  children [Modulator_7]
  properties {
   attenuationSettings=/Game/Sounds/Attenuation/Fort_Explosion_Distant_Athena_Att
   attenuationOverrides {falloffDistance=3000, attenuationShapeExtents {x=1000}}
  }
 }
 {id=Delay_0, children [Attenuation_8], properties {delayMax=0.2}}
 {id=Mixer_5, children [SoundClass_3,SoundClass_2,SoundClass_0], properties {inputVolume [1,0.75,0.75]}}
 {id=Modulator_3, children [Random_9], properties {pitchMax=1.25, pitchMin=0.85, volumeMax=1, volumeMin=1}}
 {id=Modulator_5, children [Random_11], properties {pitchMax=1.25, pitchMin=0.85}}
 {id=Modulator_7, children [Random_12], properties {pitchMax=1.25, pitchMin=0.85}}
 {
  id=Random_11
  children [WavePlayer_43, WavePlayer_44, WavePlayer_45, WavePlayer_46, WavePlayer_47, WavePlayer_48, WavePlayer_49]
  properties {weights [1,1,1,1,1,1,1]}
 }
 {id=Random_12, children [WavePlayer_50, WavePlayer_51, WavePlayer_52, WavePlayer_53, WavePlayer_54], properties {weights [1,1,1,1,1]}}
 {
  id=Random_9
  children [WavePlayer_31, WavePlayer_32, WavePlayer_33, WavePlayer_34, WavePlayer_35, WavePlayer_36, WavePlayer_37]
  properties {weights [1,1,1,1,1,1,1]}
 }
 {id=SoundClass_0, children [Delay_0], properties {soundClassOverride=&Fort_WeaponRanged_Fire_Distant}}
 {
  id=SoundClass_2
  children [Attenuation_4]
  properties {soundClassOverride=&Fort_WeaponRanged_Fire_Distant}
 }
 {
  id=SoundClass_3
  children [Attenuation_6]
  properties {soundClassOverride=&Fort_WeaponRanged_Fire_3P_PSM}
 }
 {id=WavePlayer_31, properties {soundWaveAssetPtr=&/Bow_ClusterBomb_Explo_Close_01}}
 {id=WavePlayer_32, properties {soundWaveAssetPtr=&/Bow_ClusterBomb_Explo_Close_02}}
 {id=WavePlayer_33, properties {soundWaveAssetPtr=&/Bow_ClusterBomb_Explo_Close_03}}
 {id=WavePlayer_34, properties {soundWaveAssetPtr=&/Bow_ClusterBomb_Explo_Close_04}}
 {id=WavePlayer_35, properties {soundWaveAssetPtr=&/Bow_ClusterBomb_Explo_Close_05}}
 {id=WavePlayer_36, properties {soundWaveAssetPtr=&/Bow_ClusterBomb_Explo_Close_06}}
 {id=WavePlayer_37, properties {soundWaveAssetPtr=&/Bow_ClusterBomb_Explo_Close_07}}
 {id=WavePlayer_43, properties {soundWaveAssetPtr=&/Bow_ClusterBomb_Explo_Mid_01}}
 {id=WavePlayer_44, properties {soundWaveAssetPtr=&/Bow_ClusterBomb_Explo_Mid_02}}
 {id=WavePlayer_45, properties {soundWaveAssetPtr=&/Bow_ClusterBomb_Explo_Mid_03}}
 {id=WavePlayer_46, properties {soundWaveAssetPtr=&/Bow_ClusterBomb_Explo_Mid_04}}
 {id=WavePlayer_47, properties {soundWaveAssetPtr=&/Bow_ClusterBomb_Explo_Mid_05}}
 {id=WavePlayer_48, properties {soundWaveAssetPtr=&/Bow_ClusterBomb_Explo_Mid_06}}
 {id=WavePlayer_49, properties {soundWaveAssetPtr=&/Bow_ClusterBomb_Explo_Mid_07}}
 {id=WavePlayer_50, properties {soundWaveAssetPtr=&/Bow_ClusterBomb_Explo_Distant_01}}
 {id=WavePlayer_51, properties {soundWaveAssetPtr=&/Bow_ClusterBomb_Explo_Distant_02}}
 {id=WavePlayer_52, properties {soundWaveAssetPtr=&/Bow_ClusterBomb_Explo_Distant_03}}
 {id=WavePlayer_53, properties {soundWaveAssetPtr=&/Bow_ClusterBomb_Explo_Distant_04}}
 {id=WavePlayer_54, properties {soundWaveAssetPtr=&/Bow_ClusterBomb_Explo_Distant_05}}
]
properties {
 bDebug=true
 maxDistance=27500
 volumeMultiplier=0.85
 concurrencySet [/Game/Sounds/Concurrency_Rules/Bow_Global_Conc, /Game/Sounds/Concurrency_Rules/Bow_Local_Conc]
}
