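# Sound cue for the small explosion pop of the cluster bomb bow projectile.
# A mixer blends close, mid and distant layers at descending volume, each picking randomly from four waves with modulation and its own explosion attenuation.
# The close layer uses a third-person fire sound class, while mid and distant share the distant fire class.
# The distant layer adds a short random delay, and bow global and local concurrency rules apply.
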SoundCue /PrimalGameplay/Sounds/Weapons/Bows/ClusterBomb/Bow_ClusterBomb_Prj_Explode_Small_Cue
root=Mixer_5
nodes [
 {id=Attenuation_4, children [Modulator_5], properties {attenuationSettings=/Game/Sounds/Attenuation/Fort_Explosion_Medium_Athena_Att}}
 {
  id=Attenuation_6
  children [Modulator_3]
  properties {
   attenuationSettings=/Game/Sounds/Attenuation/Fort_Explosion_Close_Athena_Att
   attenuationOverrides {falloffDistance=900, attenuationShapeExtents {x=100}}
  }
 }
 {
  id=Attenuation_8
  children [Modulator_7]
  properties {
   attenuationSettings=/Game/Sounds/Attenuation/Fort_Explosion_Distant_Athena_Att
   attenuationOverrides {falloffDistance=3000, attenuationShapeExtents {x=1000}}
  }
 }
 {id=Delay_0, children [Attenuation_8], properties {delayMax=0.2}}
 {id=Mixer_5, children [SoundClass_3,SoundClass_2,SoundClass_0], properties {inputVolume [0.9,0.5,0.3]}}
 {id=Modulator_3, children [Random_1], properties {pitchMax=1.25, pitchMin=0.85, volumeMax=1, volumeMin=1}}
 {id=Modulator_5, children [Random_3]}
 {id=Modulator_7, children [Random_2], properties {pitchMax=1.1, pitchMin=0.9}}
 {id=Random_1, children [WavePlayer_4,WavePlayer_5,WavePlayer_6,WavePlayer_7], properties {weights [1,1,1,1]}}
 {id=Random_2, children [WavePlayer_8, WavePlayer_9, WavePlayer_10, WavePlayer_11], properties {weights [1,1,1,1]}}
 {id=Random_3, children [WavePlayer_12, WavePlayer_17, WavePlayer_18, WavePlayer_19], properties {weights [1,1,1,1]}}
 {id=SoundClass_0, children [Delay_0], properties {soundClassOverride=&Fort_WeaponRanged_Fire_Distant}}
 {
  id=SoundClass_2
  children [Attenuation_4]
  properties {soundClassOverride=&Fort_WeaponRanged_Fire_Distant}
 }
 {
  id=SoundClass_3
  children [Attenuation_6]
  properties {soundClassOverride=&Fort_WeaponRanged_Fire_3P_PSM}
 }
 {id=WavePlayer_10, properties {soundWaveAssetPtr=&/Bow_ClusterBomb_Explo_Distant_Pop_03}}
 {id=WavePlayer_11, properties {soundWaveAssetPtr=&/Bow_ClusterBomb_Explo_Distant_Pop_04}}
 {id=WavePlayer_12, properties {soundWaveAssetPtr=&/Bow_ClusterBomb_Explo_Mid_Pop_01}}
 {id=WavePlayer_17, properties {soundWaveAssetPtr=&/Bow_ClusterBomb_Explo_Mid_Pop_02}}
 {id=WavePlayer_18, properties {soundWaveAssetPtr=&/Bow_ClusterBomb_Explo_Mid_Pop_03}}
 {id=WavePlayer_19, properties {soundWaveAssetPtr=&/Bow_ClusterBomb_Explo_Mid_Pop_04}}
 {id=WavePlayer_4, properties {soundWaveAssetPtr=&/Bow_ClusterBomb_Explo_Close_Pop_01}}
 {id=WavePlayer_5, properties {soundWaveAssetPtr=&/Bow_ClusterBomb_Explo_Close_Pop_02}}
 {id=WavePlayer_6, properties {soundWaveAssetPtr=&/Bow_ClusterBomb_Explo_Close_Pop_03}}
 {id=WavePlayer_7, properties {soundWaveAssetPtr=&/Bow_ClusterBomb_Explo_Close_Pop_04}}
 {id=WavePlayer_8, properties {soundWaveAssetPtr=&/Bow_ClusterBomb_Explo_Distant_Pop_01}}
 {id=WavePlayer_9, properties {soundWaveAssetPtr=&/Bow_ClusterBomb_Explo_Distant_Pop_02}}
]
properties {
 bDebug=true
 maxDistance=27500
 volumeMultiplier=0.85
 concurrencySet [/Game/Sounds/Concurrency_Rules/Bow_Global_Conc, /Game/Sounds/Concurrency_Rules/Bow_Local_Conc]
}
