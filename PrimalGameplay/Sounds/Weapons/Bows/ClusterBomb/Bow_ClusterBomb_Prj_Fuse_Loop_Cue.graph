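# Sound cue for the looping fuse sound of the cluster bomb bow projectile.
# A single looping wave player plays the projectile's in-air loop wave directly, with no mixing or randomization.
# It is classed as a ranged weapon sound, references the arrow in-air loop attenuation asset for distance falloff, and uses global and local bow concurrency rules to limit simultaneous instances.
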SoundCue /PrimalGameplay/Sounds/Weapons/Bows/ClusterBomb/Bow_ClusterBomb_Prj_Fuse_Loop_Cue
root=WavePlayer_0
nodes [{
  id=WavePlayer_0
  properties {bLooping=true, soundWaveAssetPtr=&/Bow_ClusterBomb_Prj_InAir_Loop}
 }]
properties {
 attenuationSettings=/Game/Sounds/Attenuation/Fort_Weapon_Projectile_Arrow_InAirLoop_Att
 maxDistance=16800
 soundClassObject=&Fort_WeaponRanged
 volumeMultiplier=1
 attenuationOverrides {falloffDistance=5000, reverbWetLevelMax=0.7, reverbWetLevelMin=0.2, attenuationShapeExtents {x=1000}}
 concurrencySet [/Game/Sounds/Concurrency_Rules/Bow_Global_Conc, /Game/Sounds/Concurrency_Rules/Bow_Local_Conc]
}
