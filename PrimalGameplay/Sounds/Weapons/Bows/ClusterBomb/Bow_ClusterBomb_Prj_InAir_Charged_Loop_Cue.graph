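# Sound cue for the looping in-air sound of a charged Cluster Bomb bow projectile.
# A single looping wave feeds an Enveloper that fades volume in from silence and shapes pitch with a brief rise followed by a drop.
# It uses the ranged projectile effects sound class, shared arrow in-air attenuation, and both global and local bow concurrency rules.
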SoundCue /PrimalGameplay/Sounds/Weapons/Bows/ClusterBomb/Bow_ClusterBomb_Prj_InAir_Charged_Loop_Cue
root=Enveloper_0
nodes [
 {
  id=Enveloper_0
  children [WavePlayer_0]
  properties {
   pitchCurve {editorCurveData {keys [<at -0.0003 1.049 RCIM_Cubic>, {arriveTangent=-0.0979, interpMode=RCIM_Cubic, leaveTangent=-0.0979, time=0.0804, value=1.1298}, {time=0.5}]}}
   volumeCurve {editorCurveData {keys [<at 0 0>, {arriveTangent=5.9135, interpMode=RCIM_Cubic, leaveTangent=5.9135, tangentMode=RCTM_User, time=0.1892, value=0.7977}, <at 1 1>]}}
  }
 }
 {
  id=WavePlayer_0
  properties {bLooping=true, soundWaveAssetPtr=&/Bow_ClusterBomb_Prj_InAir_Charged_Loop}
 }
]
properties {
 attenuationSettings=/Game/Sounds/Attenuation/Fort_Weapon_Projectile_Arrow_InAirLoop_Att
 maxDistance=16800
 soundClassObject=&Fort_WeaponRanged_Projectile_Effects
 volumeMultiplier=1
 attenuationOverrides {falloffDistance=5000, reverbWetLevelMax=0.7, reverbWetLevelMin=0.2, attenuationShapeExtents {x=1000}}
 concurrencySet [/Game/Sounds/Concurrency_Rules/Bow_Global_Conc, /Game/Sounds/Concurrency_Rules/Bow_Local_Conc]
}
