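# Sound cue for the cluster bomb bow's projectile spawn sound.
# A mixer layers a beep at a fixed raised pitch with a louder, randomly chosen spawn variation given a randomized upward pitch shift.
# An attenuation override applies logarithmic falloff and distance low-pass filtering.
# The cue belongs to the ranged weapon projectile effects sound class and uses global and local bow concurrency rules.
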SoundCue /PrimalGameplay/Sounds/Weapons/Bows/ClusterBomb/Bow_ClusterBomb_Prj_Spawn_Cue
root=Mixer_1
nodes [
 {id=Mixer_1, children [Modulator_2,Modulator_0], properties {inputVolume [1,1.5]}}
 {id=Modulator_0, children [Random_0], properties {pitchMax=2, pitchMin=1, volumeMax=1, volumeMin=1}}
 {id=Modulator_2, children [WavePlayer_11], properties {pitchMax=3, pitchMin=3, volumeMax=1, volumeMin=1}}
 {id=Random_0, children [WavePlayer_0,WavePlayer_1,WavePlayer_3,WavePlayer_4,WavePlayer_2], properties {weights [1,1,1,1,1]}}
 {id=WavePlayer_0, properties {soundWaveAssetPtr=&/Bow_ClusterBomb_Prj_Spawn_01}}
 {id=WavePlayer_1, properties {soundWaveAssetPtr=&/Bow_ClusterBomb_Prj_Spawn_02}}
 {id=WavePlayer_11, properties {soundWaveAssetPtr=&/Bow_ClusterBomb_Prj_Spawn_Beep_01}}
 {id=WavePlayer_2, properties {soundWaveAssetPtr=&/Bow_ClusterBomb_Prj_Spawn_05}}
 {id=WavePlayer_3, properties {soundWaveAssetPtr=&/Bow_ClusterBomb_Prj_Spawn_03}}
 {id=WavePlayer_4, properties {soundWaveAssetPtr=&/Bow_ClusterBomb_Prj_Spawn_04}}
]
properties {
 bOverrideAttenuation=true
 maxDistance=5000
 soundClassObject=&Fort_WeaponRanged_Projectile_Effects
 volumeMultiplier=1
 attenuationOverrides {
  bAttenuateWithLPF=true
  distanceAlgorithm=Logarithmic
  falloffDistance=5000
  lPFFrequencyAtMax=2000
  lPFRadiusMax=3000
  lPFRadiusMin=500
  attenuationShapeExtents {x=0}
 }
 concurrencySet [/Game/Sounds/Concurrency_Rules/Bow_Global_Conc, /Game/Sounds/Concurrency_Rules/Bow_Local_Conc]
}
