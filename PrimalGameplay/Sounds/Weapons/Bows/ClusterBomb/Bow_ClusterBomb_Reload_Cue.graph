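# Sound cue for the cluster bomb bow's reload, assembled from borrowed Metal and Stink bow samples.
# A mixer layers a random default bow reload with a random Stink knock, which follows after a fixed short delay with pitch and volume variation.
# It plays quietly as weapon foley, using shared reload attenuation plus global and local bow concurrency rules, and has its debug flag enabled.
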SoundCue /PrimalGameplay/Sounds/Weapons/Bows/ClusterBomb/Bow_ClusterBomb_Reload_Cue
root=Mixer_1
nodes [
 {id=Delay_1, children [Modulator_2], properties {delayMax=0.4, delayMin=0.4}}
 {id=Mixer_1, children [Random_3,Delay_1], properties {inputVolume [1,1.1]}}
 {id=Modulator_2, children [Random_0], properties {pitchMax=1.2, pitchMin=0.9, volumeMin=0.8}}
 {id=Random_0, children [WavePlayer_0,WavePlayer_2,WavePlayer_3,WavePlayer_4], properties {weights [1,1,1,1]}}
 {id=Random_3, children [WavePlayer_7, WavePlayer_12, WavePlayer_13], properties {weights [1,1,1]}}
 {id=WavePlayer_0, properties {soundWaveAssetPtr=/PrimalGameplay/Sounds/Weapons/Bows/Stink/Bow_Stink_Reload_Knock_01}}
 {id=WavePlayer_12, properties {soundWaveAssetPtr=/PrimalGameplay/Sounds/Weapons/Bows/Metal/Bow_Default_Reload_02}}
 {id=WavePlayer_13, properties {soundWaveAssetPtr=/PrimalGameplay/Sounds/Weapons/Bows/Metal/Bow_Default_Reload_03}}
 {id=WavePlayer_2, properties {soundWaveAssetPtr=/PrimalGameplay/Sounds/Weapons/Bows/Stink/Bow_Stink_Reload_Knock_02}}
 {id=WavePlayer_3, properties {soundWaveAssetPtr=/PrimalGameplay/Sounds/Weapons/Bows/Stink/Bow_Stink_Reload_Knock_03}}
 {id=WavePlayer_4, properties {soundWaveAssetPtr=/PrimalGameplay/Sounds/Weapons/Bows/Stink/Bow_Stink_Reload_Knock_04}}
 {id=WavePlayer_7, properties {soundWaveAssetPtr=/PrimalGameplay/Sounds/Weapons/Bows/Metal/Bow_Default_Reload_01}}
]
properties {
 attenuationSettings=&Fort_Weapon_Reload_Att
 bDebug=true
 maxDistance=2250
 soundClassObject=&Fort_WeaponFoley
 volumeMultiplier=0.2
 concurrencySet [/Game/Sounds/Concurrency_Rules/Bow_Global_Conc, /Game/Sounds/Concurrency_Rules/Bow_Local_Conc]
}
